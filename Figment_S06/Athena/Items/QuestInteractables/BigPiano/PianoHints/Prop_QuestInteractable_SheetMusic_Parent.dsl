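; Parent Blueprint for sheet music quest props in the Big Piano set, layered on a quest interactable base class.
; When the quest initializes as valid, it unhides the actor and binds sphere overlap to fade in a spotlight and activate ambient sound.
; Objective completion unbinds the overlap, then timelines scale the music stand, hide it, fade out the light and deactivate the sound.
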(event EventBeginPlay
  ;[0.5576 1 0.4597] EVENT -> BeginPlay
  (|CallParentFunction))

(event EventQuestInitializedAndValid
  ;[0.4983 1 0.4961] EVENT -> QuestInitializedAndValid
  (|CallParentFunction)
  ;[0.6632 1 0.5477] Players should this actor if they have the quest and the obj is not complete. Will be hidden completely if it is complete or they don't have the quest
  (Rendering|SetActorHiddenInGame false)
  ;[0.6967 1 0.6594] Bind if player has quest and obj is not complete
  (Collision|BindEventtoOnComponentBeginOverlap (Variables|Components|GetSphereOverlap) (EventDispatchers|CreateEvent :!event Custom|VisualsOnOverlap)))

(event EventThisQuestObjectiveComplete
  (params (IneractingPawn :type Fort%20Pawn%20Object%20Reference))
  ;[0.4983 1 0.4961] EVENT -> ThisQuestObjectiveComplete
  (|CallParentFunction IneractingPawn)
  (Development|PrintString "QuestComplete")
  ;[0.6967 1 0.6594] Unbind if the player interacts with the prop
  (Collision|UnbindEventfromOnComponentBeginOverlap (Variables|Components|GetSphereOverlap) (EventDispatchers|CreateEvent :!event Custom|VisualsOnOverlap))
  (CallFunction|HideMusic))

(event EventQuestUpdatedFromAnywhere
  ;[0.5297 1 0.5153] EVENT -> QuestUpdatedFromAnywhere
  (|CallParentFunction))

(event Custom|HideMusic
  (bind _sheetmusicstand (Variables|Components|GetSheetMusicStand))
  (bind (_direction _popanim) (Timeline|Timeline_0 PlayFromStart)
    (:Update
      (Transformation|SetWorldScale3D _sheetmusicstand (Math|Vector|Lerp(Vector) "0.010000,0.010000,0.010000" "1.000000,1.000000,1.000000" _popanim)))
    (:Finished
      ;[0.6967 1 0.6594] Hide sheet music
      (Rendering|SetVisibility _sheetmusicstand false false)
      (bind (_direction_2 _scalelight) (Timeline|Timeline_1 ReverseFromEnd)
        (:Update
          (Rendering|Components|Light|SetIntensity (Variables|Components|GetSpotlight) (* _scalelight 4e+03))
          (Components|Activation|Deactivate (Variables|Default|GetAmbientSound)))))))

(event Custom|VisualsOnOverlap
  (params (OverlappedComponent :type Primitive%20Component%20Object%20Reference) (OtherActor :type Actor%20Object%20Reference) (OtherComp :type Primitive%20Component%20Object%20Reference) (OtherBodyIndex :type Integer) (bFromSweep :type Boolean) (SweepResult :type Hit%20Result%20Structure%20%28by%20ref%29))
  ;[0.7088 1 0.5374] Turn on light & sound
  (Utilities|FlowControl|Sequence
    (:then_0
      (bind (_direction _scalelight) (Timeline|Scale%20Light PlayFromStart)
        (:Update
          (Rendering|Components|Light|SetIntensity (Variables|Components|GetSpotlight) (* _scalelight 4e+03)))))
    (:then_1
      (Components|Activation|Activate (Variables|Default|GetAmbientSound) false))))

(event EventHideQuestProp
  ;[0.5297 1 0.5153] CUSTOM EVENT -> HideQuestProp
  (|CallParentFunction))

(event Collision|EventActorBeginOverlap
  (params (OtherActor :type Actor%20Object%20Reference))
  (|CallParentFunction OtherActor) ; This node is disabled and will not be called.\nDrag off pins to build functionality.
)

(event EventTick
  (params (DeltaSeconds :type Float%20%28single-precision%29))
  (|CallParentFunction DeltaSeconds) ; This node is disabled and will not be called.\nDrag off pins to build functionality.
)

(fn ConstructionScript
  (|CallParentFunction))
